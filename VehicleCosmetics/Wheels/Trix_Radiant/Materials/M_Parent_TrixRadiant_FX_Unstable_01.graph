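# Parent translucent unlit material for an unstable energy effect on the Trix Radiant vehicle wheel cosmetic, with a tintable emissive glow.
# Opacity sums two differently panned caustics samples, one squared, masked by a shockwave curve texture that also adds a faint base.
# World position offset continuously spins vertices about the object pivot on a parameterized local axis, wrapping time each revolution.
# Results pass through shared vehicle damage and de-rez transition functions.
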Material /VehicleCosmetics/Wheels/Trix_Radiant/Materials/M_Parent_TrixRadiant_FX_Unstable_01
graph {
 properties {
  bAutomaticallySetUsageInEditor=false
  bUseMaterialAttributes=true
  bUsedWithSkeletalMesh=true
  blendMode=BLEND_Translucent
  shadingModel=MSM_Unlit
  usedShadingModels=Unlit
 }
 nodes [
  n1 = Multiply(
   A=VectorParameter() {parameterName=EmissiveHeat, defaultValue=<rgba 0 1 0.8384 1>}.RGB
   B=ScalarParameter() {defaultValue=32, parameterName=EmissiveMultiply}
  )
  n2 = TextureCoordinate() {vTiling=0.1}
  n3 = Multiply(A=Time(), B=ScalarParameter() {defaultValue=1, parameterName=PanSpeedAdjust})
  n4 = TextureSample(UVs=Panner(Coordinate=@n2, Time=@n3) {speedX=1.4, speedY=0.0675}) {samplerSource=SSM_Wrap_WorldGroupSettings, texture=/VehicleCosmetics/Wheels/_SharedTextures/FX/Caustics_D}
  n5 = Multiply(A=@n4.G, B=@n4.G)
  n6 = TextureSample(UVs=Panner(Coordinate=@n2, Time=@n3) {speedX=1, speedY=1}) {samplerSource=SSM_Wrap_WorldGroupSettings, texture=&Caustics_D}
  n7 = TextureSample() {texture=/VehicleCosmetics/Wheels/_SharedTextures/FX/Curve_Shockwaves_Pack}
  n8 = Add(A=Multiply(A=Add(A=@n5, B=@n6.R), B=@n7.R), B=Multiply(A=@n7.R, B=Constant() {r=0.025}))
  n9 = ObjectRadius()
  n10 = Transform(None=VectorParameter() {parameterName=WPORotateAxis, defaultValue=<rgba 0 1 0 1>}.RGB) {transformSourceType=TRANSFORMSOURCE_Local}
  n11 = Time()
  n12 = ScalarParameter() {defaultValue=1, parameterName=WPOSpeed}
  n13 = Divide(A=Constant() {r=6.28}, B=@n12)
  n14 = RotateAboutAxis(
   NormalizedRotationAxis=Normalize(VectorInput=Multiply(A=@n9, B=@n10))
   RotationAngle=Multiply(A=Fmod(A=@n11, B=@n13), B=@n12)
   PivotPoint=MaterialFunctionCall() {materialFunction=&ObjectPivotPoint}.Object Pivot Location
   Position=WorldPosition().XYZ
  )
  MaterialFunctionCall(In=MaterialFunctionCall(Material=SetMaterialAttributes!(MaterialAttributes, Emissive Color=@n1, Opacity=@n8, World Position Offset=@n14) {attributeSetTypes [#1, #2, #3]}) {materialFunction=&MF_BaseVehicle_Damage}.Result) {materialFunction=&MF_VehicleDeRez_Transition}
 ]
}
