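# Parent material for the Trix Radiant wheel effect: a translucent, unlit, glowing animated trail for skeletal meshes.
# Rotated, panning smoke and beam-mask textures blend CustomColor toward a brightened copy, with the beam adding extra emissive glow.
# Opacity combines an advanced Fresnel edge fade (borrowed from the CarCar_Holographic wheel), the smoke's green channel and CustomColor alpha.
# The material attributes then pass through the shared vehicle damage and de-rez transition functions.
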Material /VehicleCosmetics/Wheels/Trix_Radiant/Materials/M_WheelParent_FX_TrixRadiant
graph {
 properties {
  bAutomaticallySetUsageInEditor=false
  bUseMaterialAttributes=true
  bUsedWithSkeletalMesh=true
  blendMode=BLEND_Translucent
  shadingModel=MSM_Unlit
  usedShadingModels=Unlit
 }
 nodes [
  CameraVectorWS()
  n1 = VectorParameter() {parameterName=CustomColor, defaultValue=<rgba 1 0.05 0.01 2>}
  n2 = Multiply(A=@n1.RGB, B=Constant() {r=8})
  n3 = TextureCoordinate()
  n4 = Constant() {r=0.5}
  n5 = TextureSample(UVs=Panner(Coordinate=Rotator(Coordinate=@n3, Time=@n4) {speed=3.14}) {speedX=-1}) {samplerType=SAMPLERTYPE_LinearColor, texture=/VehicleCosmetics/Wheels/Trix_Radiant/Textures/Smoke_Trail_02_T}
  n6 = Rotator(Coordinate=TextureCoordinate() {uTiling=2, vTiling=0.2}, Time=Constant() {r=0.5}) {speed=3.14}
  n7 = Multiply(
   A=TextureSample(UVs=Panner(Coordinate=@n6) {speedX=-0.5, speedY=1}) {samplerType=SAMPLERTYPE_LinearColor, texture=/VehicleCosmetics/Wheels/Trix_Radiant/Textures/Beam_GlowMask_T}.R
   B=Constant() {r=4}
  )
  n8 = Add(A=LinearInterpolate(A=@n1.RGB, B=@n2, Alpha=Add(A=@n5.R, B=@n7)), B=Multiply(A=@n7, B=@n2))
  n9 = DotProduct(
   A=MaterialFunctionCall!(Input Normal, Input Sharpness=ScalarParameter() {defaultValue=1.2}, Input Two Sided) {materialFunction=/VehicleCosmetics/Wheels/CarCar_Holographic/Materials/MF_Fresnel_Advanced}.Middle
   B=Constant3Vector() {constant=<rgba 0 0 1 0>}
  )
  n10 = ScalarParameter() {defaultValue=3, parameterName=EdgeFade}
  n11 = Multiply(A=Clamp(None=Power(Base=@n9, Exp=@n10)), B=Multiply(A=@n5.G, B=Constant() {r=2}))
  n12 = Clamp(None=Multiply(A=Clamp(None=@n11), B=@n1.A))
  MaterialFunctionCall(In=MaterialFunctionCall(Material=SetMaterialAttributes!(MaterialAttributes, Emissive Color=@n8, Opacity=@n12) {attributeSetTypes [#1, #2]}) {materialFunction=&MF_BaseVehicle_Damage}.Result) {materialFunction=&MF_VehicleDeRez_Transition}
 ]
}
