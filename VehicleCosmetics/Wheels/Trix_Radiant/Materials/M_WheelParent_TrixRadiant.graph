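# Parent material for the Trix Radiant vehicle wheel, adding a pulsing two-color emissive glow.
# A sine-over-time pulse multiplies a Fresnel term and the rim texture's red and green masks, tinted between AccentColor and dimmed CustomColor.
# Inverted red vertex color blends this glow into MF_WheelParent_Simple's emissive.
# The resulting material attributes pass through the shared vehicle damage and de-rez transition functions.
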Material /VehicleCosmetics/Wheels/Trix_Radiant/Materials/M_WheelParent_TrixRadiant
graph {
 properties {
  bAutomaticallySetUsageInEditor=false
  bUseMaterialAttributes=true
  bUsedWithMeshDeformer=true
  bUsedWithSkeletalMesh=true
  pixelDepthOffsetMode=PDOM_Legacy
  previewMesh=/VehicleCosmetics/Wheels/Trix_Radiant/Meshes/SK_Wheel_TrixRadiant
  usedShadingModels=Default Lit
 }
 nodes [
  CameraVectorWS()
  n1 = MaterialFunctionCall() {materialFunction=&MF_WheelParent_Simple}
  n2 = GetMaterialAttributes(None=@n1.MaterialAttributes) {attributeGetTypes [#1]}
  n3 = LinearInterpolate(A=@n2.Emissive Color, B=NamedRerouteUsage(), Alpha=OneMinus(None=@n1.VertexColorR))
  MaterialFunctionCall(In=MaterialFunctionCall(Material=SetMaterialAttributes(MaterialAttributes=@n2.MaterialAttributes, Emissive Color=@n3) {attributeSetTypes [#1]}) {materialFunction=&MF_BaseVehicle_Damage}.Result) {materialFunction=&MF_VehicleDeRez_Transition}
  n4 = Time()
  n5 = Constant() {r=2.5}
  n6 = Add(A=Sine(None=Multiply(A=@n4, B=@n5)), B=Constant() {r=1.2})
  n7 = MaterialFunctionCall!(Input Normal, Input Sharpness, Input Two Sided) {materialFunction=/VehicleCosmetics/Wheels/CarCar_Holographic/Materials/MF_Fresnel_Advanced}
  n8 = Clamp(None=Power(Base=ComponentMask(None=@n7.Middle) {b=true}, Exp=Constant() {r=5}))
  n9 = Multiply(A=@n6, B=Clamp(None=ConstantBiasScale(None=@n8) {bias=0, scale=0.25}) {maxDefault=1.5})
  n10 = TextureSampleParameter2D() {parameterName=Masks, samplerType=SAMPLERTYPE_LinearColor, texture=/VehicleCosmetics/Wheels/Trix_Radiant/Textures/T_Trix_Radiant_Rim_M}
  n11 = Add(A=Multiply(A=@n10.G, B=@n6), B=Multiply(A=@n10.R, B=@n6))
  n12 = VectorParameter() {parameterName=AccentColor, defaultValue=<rgba 0.9734 0 0.6867 1>}
  n13 = Multiply(A=VectorParameter() {parameterName=CustomColor, defaultValue=<rgba 0 0.9734 0.3231 1>}.RGB, B=Constant() {r=0.25})
  n14 = NamedRerouteUsage()
  NamedRerouteDeclaration(None=Multiply(A=Multiply(A=@n9, B=@n11), B=LinearInterpolate(A=@n12.RGB, B=@n13, Alpha=@n14))) {name=Emissive, nodeColor=<rgba 1 0.75 0 1>}
  NamedRerouteDeclaration(None=@n10.R) {name=Red_CM, nodeColor=<rgba 1 0.3047 0 1>}
  NamedRerouteDeclaration(None=@n10.G) {name=Green_CM, nodeColor=<rgba 1 0.3047 0 1>}
 ]
}
