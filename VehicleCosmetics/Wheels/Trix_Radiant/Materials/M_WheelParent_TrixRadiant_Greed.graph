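# Parent material for the Trix Radiant Greed cosmetic wheel, an unlit, masked, animated glow for skeletal meshes.
# A tiled noise texture scrolls slowly along V and is multiplied by an offset sine of time to pulse.
# That value, tinted by CustomColor, drives emissive and, scaled up, also the opacity mask, so the cutout pulses with the glow.
# The attributes then pass through the shared vehicle damage and de-rez transition functions.
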Material /VehicleCosmetics/Wheels/Trix_Radiant/Materials/M_WheelParent_TrixRadiant_Greed
graph {
 properties {
  bAutomaticallySetUsageInEditor=false
  bUseMaterialAttributes=true
  bUsedWithSkeletalMesh=true
  blendMode=BLEND_Masked
  shadingModel=MSM_Unlit
  usedShadingModels=Unlit
 }
 nodes [
  n1 = TextureCoordinate() {uTiling=5, vTiling=5}
  n2 = Multiply(A=Time(), B=Constant() {r=1.7})
  n3 = Multiply(
   A=TextureSample(UVs=Panner(Coordinate=@n1) {speedY=0.1}) {samplerType=SAMPLERTYPE_LinearColor, texture=/VehicleCosmetics/Wheels/Trix_Radiant/Textures/Universal_Noise_Heightmap_02}.R
   B=Add(A=Sine(None=@n2), B=Constant() {r=1.3})
  )
  n4 = Multiply(A=VectorParameter() {parameterName=CustomColor, defaultValue=<rgba 0.776 0 0.3377 1>}.RGB, B=@n3)
  n5 = Multiply(A=@n3, B=Constant() {r=6})
  MaterialFunctionCall(In=MaterialFunctionCall(Material=SetMaterialAttributes!(MaterialAttributes, Emissive Color=@n4, Opacity Mask=@n5) {attributeSetTypes [#1, #2]}) {materialFunction=&MF_BaseVehicle_Damage}.Result) {materialFunction=&MF_VehicleDeRez_Transition}
 ]
}
